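# Niagara system named as the idle effect for the StreetBee melee pickaxe, with four glow emitters.
# The emitters are named for the top and bottom positions on the front and back.
# SystemState is the only listed module, and no renderers, data interfaces, dynamic inputs or user variables appear.
# The LWE suffix suggests lightweight emitters, whose internals are not visible here.
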NiagaraSystem /BRCosmetics/Weapons/FORT_Melee/Pickaxe_StreetBee/FX/NS_Pickaxe_StreetBee_Idle_LWE
dependencies {usedDataInterfaces [], usedDynamicInputs [], usedModules [&SystemState]}
emitters {
 GlowBottomBack {rendererClasses [], renderers []}
 GlowBottomFront {rendererClasses [], renderers []}
 GlowTopBack {rendererClasses [], renderers []}
 GlowTopFront {rendererClasses [], renderers []}
}
summary {
 systemName=NS_Pickaxe_StreetBee_Idle_LWE
 emitters [
  {emitterName=GlowTopFront, rendererClasses []}
  {emitterName=GlowBottomFront, rendererClasses []}
  {emitterName=GlowTopBack, rendererClasses []}
  {emitterName=GlowBottomBack, rendererClasses []}
 ]
 userVariables []
}
